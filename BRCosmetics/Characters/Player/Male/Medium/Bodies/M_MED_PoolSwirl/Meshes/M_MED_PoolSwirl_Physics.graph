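# Physics asset for the PoolSwirl medium male player body cosmetic, set up for its dangling belt strap, beads and front pad.
# Capsules on pelvis and thighs cover the body; the belt strap chain hangs from pelvis, the bead chain from left thigh.
# Strap and bead links have narrow swing limits and near-zero twist, with beads tightest.
# The boxed front pad, its right-side capsule and the chain roots use free angular constraints.
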PhysicsAsset /BRCosmetics/Characters/Player/Male/Medium/Bodies/M_MED_PoolSwirl/Meshes/M_MED_PoolSwirl_Physics
bodies {
 base_beads_0 [{shapeType=Capsule, length=0.505, radius=0.505, shapeName=base_beads_0_capsule, rotation=<rot 0 -90 0>}]
 base_beltStrap_00 [{shapeType=Capsule, length=0.505, radius=0.505, shapeName=base_beltStrap_00_capsule, rotation=<rot 0 -90 0>}]
 dyn_beads_01 [{shapeType=Capsule, length=2, radius=1, shapeName=dyn_beads_01_capsule, center=<2 0 0>, rotation=<rot 90 0 0>}]
 dyn_beads_02 [{shapeType=Capsule, length=2, radius=1, shapeName=dyn_beads_02_capsule, center=<2 0 0>, rotation=<rot 90 0 0>}]
 dyn_beltStrap_00 [{shapeType=Capsule, length=4.5574, radius=1.8832, shapeName=dyn_beltStrap_00_capsule, center=<4.0771 -1.3868 0>, rotation=<rot 90 0 0>}]
 dyn_beltStrap_01 [{shapeType=Capsule, length=3.3024, radius=2.0808, shapeName=dyn_beltStrap_01_capsule, center=<3.049 -1.2836 0>, rotation=<rot 90 0 0>}]
 dyn_beltStrap_02 [{shapeType=Capsule, length=3.6257, radius=2.0808, shapeName=dyn_beltStrap_02_capsule, center=<3.515 -1.231 0>, rotation=<rot 90 0 0>}]
 dyn_beltStrap_03 [{shapeType=Capsule, length=1.7888, radius=2.6031, shapeName=dyn_beltStrap_03_capsule, center=<3.5615 -1.1724 0>, rotation=<rot 90 0 0>}]
 dyn_frontPad_01 [{
   shapeType=Box
   extentX=2.7535
   extentY=13.4965
   extentZ=15.9121
   shapeName=dyn_frontPad_01_box
   center=<9.3075 0.4128 -0.1565>
   rotation=<rot 0.0474 -89.9546 -89.2274>
  }]
 dyn_frontPad_r [{
   shapeType=Capsule
   length=10.3499
   radius=1.654
   shapeName=dyn_frontPad_r_capsule
   center=<5.6264 0.9107 1.5655>
   rotation=<rot 79.9555 179.9994 -178.5684>
  }]
 pelvis [{shapeType=Capsule, length=6.8965, radius=10.9907, shapeName=pelvis_capsule, center=<0 -1.1567 0>}]
 thigh_l [{
   shapeType=Capsule
   length=26.3157
   radius=11.2641
   shapeName=thigh_l_capsule
   center=<-22.7273 -1.3105 1.555>
   rotation=<rot 88.3875 0.0002 0.0006>
  }]
 thigh_r [{
   shapeType=Capsule
   length=25.9368
   radius=11.6431
   shapeName=thigh_l_capsule
   center=<22.774 1.3105 -3.3731>
   rotation=<rot 88.5272 179.999 179.9992>
  }]
}
constraints [
 {
  bone1Name=thigh_l
  bone2Name=pelvis
  swing1LimitDegrees=45
  swing1Motion=Free
  swing2LimitDegrees=45
  swing2Motion=Free
  twistLimitDegrees=45
  twistMotion=Free
 }
 {
  bone1Name=thigh_r
  bone2Name=pelvis
  swing1LimitDegrees=45
  swing1Motion=Free
  swing2LimitDegrees=45
  swing2Motion=Free
  twistLimitDegrees=45
  twistMotion=Free
 }
 {
  bone1Name=base_beltStrap_00
  bone2Name=pelvis
  swing1LimitDegrees=45
  swing1Motion=Free
  swing2LimitDegrees=45
  swing2Motion=Free
  twistLimitDegrees=45
  twistMotion=Free
 }
 {bone1Name=dyn_beltStrap_00, bone2Name=base_beltStrap_00, swing1LimitDegrees=25, swing2LimitDegrees=25, twistLimitDegrees=1}
 {bone1Name=dyn_beltStrap_01, bone2Name=dyn_beltStrap_00, swing1LimitDegrees=30, swing2LimitDegrees=30, twistLimitDegrees=1}
 {bone1Name=dyn_beltStrap_02, bone2Name=dyn_beltStrap_01, swing1LimitDegrees=30, swing2LimitDegrees=30, twistLimitDegrees=1}
 {bone1Name=dyn_beltStrap_03, bone2Name=dyn_beltStrap_02, swing1LimitDegrees=30, swing2LimitDegrees=30, twistLimitDegrees=1}
 {
  bone1Name=dyn_frontPad_r
  bone2Name=dyn_frontPad_01
  swing1LimitDegrees=45
  swing1Motion=Free
  swing2LimitDegrees=45
  swing2Motion=Free
  twistLimitDegrees=45
  twistMotion=Free
 }
 {
  bone1Name=base_beads_0
  bone2Name=thigh_l
  swing1LimitDegrees=45
  swing1Motion=Free
  swing2LimitDegrees=45
  swing2Motion=Free
  twistLimitDegrees=45
  twistMotion=Free
 }
 {bone1Name=dyn_beads_01, bone2Name=base_beads_0, swing1LimitDegrees=1.5, swing2LimitDegrees=5, twistLimitDegrees=0}
 {bone1Name=dyn_beads_02, bone2Name=dyn_beads_01, swing1LimitDegrees=2, swing2LimitDegrees=10, twistLimitDegrees=0}
 {
  bone1Name=dyn_frontPad_01
  bone2Name=pelvis
  swing1LimitDegrees=45
  swing1Motion=Free
  swing2LimitDegrees=45
  swing2Motion=Free
  twistLimitDegrees=45
  twistMotion=Free
 }
]
