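# Physics asset for the left sleeve and arm straps of the PoolSwirl medium male character body.
# Capsules cover the lower spine, left arm, sleeve bones and two-segment front and back straps; upperarm_l duplicates the sleeveMaster capsule.
# The back strap and lower sleeve attach to upperarm_l, while the front strap chains from sleeveMaster.
# These have tight swing and near-zero twist limits; body, arm and sleeveMaster joints rotate freely.
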PhysicsAsset /BRCosmetics/Characters/Player/Male/Medium/Bodies/M_MED_PoolSwirl/Meshes/M_MED_PoolSwirl_Sleeve_Physics
bodies {
 dyn_armStrap_00_back [{
   shapeType=Capsule
   length=4.3914
   radius=1
   shapeName=dyn_armStrap_00_back_capsule
   center=<3.0969 -0.8361 -0.2501>
   rotation=<rot 87.3699 -178.9376 177.9096>
  }]
 dyn_armStrap_00_front [{
   shapeType=Capsule
   length=4.7236
   radius=1
   shapeName=dyn_armStrap_00_front_capsule
   center=<3.3003 -1.0276 0>
   rotation=<rot 90 180 172.0501>
  }]
 dyn_armStrap_01_back [{
   shapeType=Capsule
   length=4.982
   radius=1.3676
   shapeName=dyn_armStrap_01_back_capsule
   center=<3.5954 -0.2294 -0.0768>
   rotation=<rot 90 28.5303 26.2812>
  }]
 dyn_armStrap_01_front [{shapeType=Capsule, length=5.5033, radius=1.4105, shapeName=dyn_armStrap_01_front_capsule, center=<3.9543 -0.8027 0>, rotation=<rot 90 0 0>}]
 dyn_sleeveLower [{
   shapeType=Capsule
   length=11.5321
   radius=7.0414
   shapeName=dyn_sleeveLower_capsule
   center=<23.9204 -0.6754 -4.9453>
   rotation=<rot 77.1189 -0.0005 -0.0006>
  }]
 dyn_sleeveMaster [{
   shapeType=Capsule
   length=12.4919
   radius=8.3411
   shapeName=dyn_sleeveMaster_capsule
   center=<16.8995 -0.2496 -3.1556>
   rotation=<rot 82.0972 -0.0485 -0.0559>
  }]
 lowerarm_l [{shapeType=Capsule, length=16.1954, radius=5.6936, shapeName=lowerarm_l_capsule, center=<12.0489 0.289 1.6852>, rotation=<rot 90 0 0>}]
 spine_01 [{shapeType=Capsule, length=2.6691, radius=11.696, shapeName=spine_01_capsule, center=<8.3231 0 0>}]
 spine_02 [{shapeType=Capsule, length=4.8192, radius=11.696, shapeName=spine_02_capsule, center=<9.2861 0 0>}]
 upperarm_l [{
   shapeType=Capsule
   length=12.4919
   radius=8.3411
   shapeName=dyn_sleeveMaster_capsule
   center=<16.8995 -0.2496 -3.1556>
   rotation=<rot 82.0972 -0.0485 -0.0559>
  }]
}
constraints [
 {
  bone1Name=lowerarm_l
  bone2Name=upperarm_l
  swing1LimitDegrees=45
  swing1Motion=Free
  swing2LimitDegrees=45
  swing2Motion=Free
  twistLimitDegrees=45
  twistMotion=Free
 }
 {
  bone1Name=dyn_sleeveMaster
  bone2Name=upperarm_l
  swing1LimitDegrees=45
  swing1Motion=Free
  swing2LimitDegrees=45
  swing2Motion=Free
  twistLimitDegrees=45
  twistMotion=Free
 }
 {bone1Name=dyn_armStrap_00_back, bone2Name=upperarm_l, swing1LimitDegrees=40, swing2LimitDegrees=25, twistLimitDegrees=1}
 {bone1Name=dyn_armStrap_01_back, bone2Name=dyn_armStrap_00_back, swing1LimitDegrees=35, swing2LimitDegrees=25, twistLimitDegrees=1}
 {bone1Name=dyn_armStrap_00_front, bone2Name=dyn_sleeveMaster, swing1LimitDegrees=25, swing2LimitDegrees=25, twistLimitDegrees=1}
 {bone1Name=dyn_armStrap_01_front, bone2Name=dyn_armStrap_00_front, swing1LimitDegrees=25, swing2LimitDegrees=25, twistLimitDegrees=1}
 {bone1Name=dyn_sleeveLower, bone2Name=upperarm_l, swing1LimitDegrees=12, swing2LimitDegrees=8, twistLimitDegrees=1}
 {
  bone1Name=upperarm_l
  bone2Name=spine_02
  swing1LimitDegrees=45
  swing1Motion=Free
  swing2LimitDegrees=45
  swing2Motion=Free
  twistLimitDegrees=45
  twistMotion=Free
 }
 {
  bone1Name=spine_02
  bone2Name=spine_01
  swing1LimitDegrees=45
  swing1Motion=Free
  swing2LimitDegrees=45
  swing2Motion=Free
  twistLimitDegrees=45
  twistMotion=Free
 }
]
