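# Physics asset for the sword and sword string on the M_MED_PoolSwirl male player body, using one capsule per bone.
# The sword is constrained to the pelvis with limited swing and locked twist.
# A string base joins the sword with free rotation and roots a three-segment string chain with tight swing and twist limits.
# Right upper arm, forearm and hand capsules chain from the pelvis with free rotation, likely serving as colliders.
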PhysicsAsset /BRCosmetics/Characters/Player/Male/Medium/Bodies/M_MED_PoolSwirl/Meshes/M_MED_PoolSwirl_Sword_Physics
bodies {
 base_swordString_00 [{shapeType=Capsule, length=0.505, radius=0.505, shapeName=base_swordString_00_capsule, rotation=<rot 0 -90 0>}]
 dyn_sword [{
   shapeType=Capsule
   length=30.583
   radius=2.7808
   shapeName=dyn_sword_capsule
   center=<22.396 2.4579 -0.2487>
   rotation=<rot 85.6794 36.9088 35.7748>
  }]
 dyn_swordString_00 [{shapeType=Capsule, length=3.0062, radius=1, shapeName=dyn_swordString_00_capsule, center=<2.4036 0 0>, rotation=<rot 90 0 0>}]
 dyn_swordString_01 [{shapeType=Capsule, length=2.2939, radius=1, shapeName=dyn_swordString_01_capsule, center=<1.9017 0 0>, rotation=<rot 90 0 0>}]
 dyn_swordString_02 [{shapeType=Capsule, length=3.5453, radius=1.5594, shapeName=dyn_swordString_02_capsule, center=<3.0852 0 0>, rotation=<rot 90 0 0>}]
 hand_r [{shapeType=Capsule, length=5.4988, radius=4.4167, shapeName=hand_r_capsule, center=<-4.713 1.3609 0>, rotation=<rot 90 0 0>}]
 lowerarm_r [{shapeType=Capsule, length=18.3754, radius=5.0316, shapeName=lowerarm_r_capsule, center=<-9.0374 0 0>, rotation=<rot 90 0 0>}]
 pelvis [{shapeType=Capsule, length=2, radius=1, shapeName=pelvis_capsule, center=<1.5885 -1.1783 -0.5032>, rotation=<rot 4.3507 -90 6.5083>}]
 upperarm_r [{shapeType=Capsule, length=15.5772, radius=7.0849, shapeName=upperarm_r_capsule, center=<-11.8928 0.3937 0>, rotation=<rot 90 0 0>}]
}
constraints [
 {bone1Name=dyn_sword, bone2Name=pelvis, swing1LimitDegrees=25, swing2LimitDegrees=35, twistLimitDegrees=0, twistMotion=Locked}
 {
  bone1Name=base_swordString_00
  bone2Name=dyn_sword
  swing1LimitDegrees=45
  swing1Motion=Free
  swing2LimitDegrees=45
  swing2Motion=Free
  twistLimitDegrees=45
  twistMotion=Free
 }
 {bone1Name=dyn_swordString_00, bone2Name=base_swordString_00, swing1LimitDegrees=25, swing2LimitDegrees=25, twistLimitDegrees=1}
 {bone1Name=dyn_swordString_01, bone2Name=dyn_swordString_00, swing1LimitDegrees=25, swing2LimitDegrees=25, twistLimitDegrees=1}
 {bone1Name=dyn_swordString_02, bone2Name=dyn_swordString_01, swing1LimitDegrees=25, swing2LimitDegrees=25, twistLimitDegrees=1}
 {
  bone1Name=upperarm_r
  bone2Name=pelvis
  swing1LimitDegrees=45
  swing1Motion=Free
  swing2LimitDegrees=45
  swing2Motion=Free
  twistLimitDegrees=45
  twistMotion=Free
 }
 {
  bone1Name=lowerarm_r
  bone2Name=upperarm_r
  swing1LimitDegrees=45
  swing1Motion=Free
  swing2LimitDegrees=45
  swing2Motion=Free
  twistLimitDegrees=45
  twistMotion=Free
 }
 {
  bone1Name=hand_r
  bone2Name=lowerarm_r
  swing1LimitDegrees=45
  swing1Motion=Free
  swing2LimitDegrees=45
  swing2Motion=Free
  twistLimitDegrees=45
  twistMotion=Free
 }
]
