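# Animation Blueprint for the PoolSwirl male medium cosmetic body, layering secondary motion onto the pose copied from its parent mesh.
# Cached poses chain Control Rig, bone constraints, state-driven offsets, AnimDynamics and RigidBody physics for straps, pants, sleeve and sword, then fixed correctives.
# The event graph derives blend alphas from movement, crouch, front-end and cape state, including rigid-body-off fallbacks.
# It also hides a material element when a PoolSwirl-tagged pickaxe is equipped.
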AnimBlueprint /BRCosmetics/Characters/Player/Male/Medium/Bodies/M_MED_PoolSwirl/Meshes/M_MED_PoolSwirl_AnimBP
target_skeleton=&/M_MED_PoolSwirl_Skeleton
graphs [
 AnimGraph {nodes [
   Output Pose(Result=Use cached pose 'Post Correctives'()) {layerGroup=None}
   n1 = Control Rig!(
    Source=Copy Pose From Mesh() {bUseAttachedParent=true}
    Alpha
    is Crouching Alpha:double=Get is Crouching CR() -> (is Crouching CR:double)
    FrontEndIdle:double=Get FrontendIdle() -> (FrontendIdle:double)
   ) -> (Pose) {
    alphaBoolBlend {blendOption=Linear}
    controlRigAssetReference {blueprintRigClass=&/M_MED_PoolSwirl_CtrlRig.M_MED_PoolSwirl_CtrlRig_C}
   }
   n2 = "Constraint - base_legBag_00 (PPPPP)"!(
    ComponentPose=Local To Component(LocalPose=@n1)
    Alpha=1
    ConstraintWeights_0=0.7168
    ConstraintWeights_1:float=0.1963
    ConstraintWeights_2:float=0.0443
    ConstraintWeights_3:float=0.0326
    ConstraintWeights_4:float=0.01
   ) -> (Pose) {
    alphaBoolBlend {blendOption=Linear}
    boneToModify {boneName=base_legBag_00}
    constraintSetup [
     {targetBone {boneName=thigh_twist_01_r}}
     {targetBone {boneName=pelvis}}
     {targetBone {boneName=deform_groin_r}}
     {targetBone {boneName=deform_glute_r}}
     {targetBone {boneName=thigh_r}}
    ]
    constraintWeights [1,1,1,1,1]
   }
   n3 = "Constraint - base_backStrap (PPPPPPPP)"!(
    ComponentPose=@n2
    Alpha=1
    ConstraintWeights_0=0.4381
    ConstraintWeights_1:float=0.2812
    ConstraintWeights_2:float=0.1384
    ConstraintWeights_3:float=0.0443
    ConstraintWeights_4:float=0.0414
    ConstraintWeights_5:float=0.0309
    ConstraintWeights_6:float=0.0244
    ConstraintWeights_7:float=0.0014
   ) -> (Pose) {
    alphaBoolBlend {blendOption=Linear}
    boneToModify {boneName=base_backStrap}
    constraintSetup [
     {targetBone {boneName=clavicle_r}}
     {targetBone {boneName=spine_05}}
     {targetBone {boneName=spine_03}}
     {targetBone {boneName=spine_04}}
     {targetBone {boneName=upperarm_twist_01_r}}
     {targetBone {boneName=deform_upperarm_bk_r}}
     {targetBone {boneName=deform_upperarm_up_r}}
     {targetBone {boneName=upperarm_twist_02_r}}
    ]
    constraintWeights [1,1,1,1,1,1,1,1]
   }
   Save cached pose 'Start Pose'(Pose=Component To Local(ComponentPose=@n3)) {cacheName=Start Pose}
   n4 = Anim Dynamics - Bone: dyn_legStrap_01(ComponentPose=Local To Component(LocalPose=Use cached pose 'Pre Dynamics'()), Alpha=1, ExternalForce="0,0,0") {
    angularDampingOverride=1
    angularSpringConstant=20000
    bAngularSpring=true
    bOverrideAngularDamping=true
    bUseGravityOverride=true
    gravityScale=0.35
    simulationSpace=BoneRelative
    alphaBoolBlend {blendOption=Linear}
    boundBone {boneName=dyn_legStrap_01}
    componentLinearAccScale=<1 1 1>
    componentLinearVelScale=<1 1 1>
    gravityOverride {z=-980}
    physicsBodyDefinitions [{
      boundBone {boneName=dyn_legStrap_01}
      boxExtents=<20 30 30>
      constraintSetup {angularLimitsMax=<0 15 15>, angularLimitsMin=<0 -15 -30>, angularTarget=<1 0 0>}
      localJointOffset=<2 0 0>
     }]
    relativeSpaceBone {boneName=pelvis}
   }
   n5 = Anim Dynamics - Bone: dyn_legStrap_00(ComponentPose=@n4, Alpha=1, ExternalForce="0,0,0") {
    angularDampingOverride=1
    angularSpringConstant=20000
    bAngularSpring=true
    bOverrideAngularDamping=true
    bUseGravityOverride=true
    gravityScale=0.35
    simulationSpace=BoneRelative
    alphaBoolBlend {blendOption=Linear}
    boundBone {boneName=dyn_legStrap_00}
    componentLinearAccScale=<1 1 1>
    componentLinearVelScale=<1 1 1>
    gravityOverride {z=-980}
    physicsBodyDefinitions [{
      boundBone {boneName=dyn_legStrap_00}
      boxExtents=<20 30 30>
      constraintSetup {angularLimitsMax=<0 15 15>, angularLimitsMin=<0 -15 -30>, angularTarget=<1 0 0>}
      localJointOffset=<2 0 0>
     }]
    relativeSpaceBone {boneName=pelvis}
   }
   n6 = Anim Dynamics - Bone: dyn_backStrap(ComponentPose=@n5, Alpha=1, ExternalForce="0,0,0") {
    angularDampingOverride=1
    angularSpringConstant=20000
    bAngularSpring=true
    bOverrideAngularDamping=true
    bUseGravityOverride=true
    gravityScale=0.35
    simulationSpace=RootRelative
    alphaBoolBlend {blendOption=Linear}
    boundBone {boneName=dyn_backStrap}
    componentLinearAccScale=<1 1 1>
    componentLinearVelScale=<1 1 1>
    gravityOverride {z=-980}
    physicsBodyDefinitions [{
      boundBone {boneName=dyn_backStrap}
      boxExtents=<20 30 30>
      constraintSetup {angularLimitsMax=<0 15 1>, angularLimitsMin=<-10 -15 -50>, angularTarget=<1 0 0>}
      localJointOffset=<4 0 0>
     }]
   }
   n7 = Anim Dynamics - Bone: dyn_StringRarm_00(ComponentPose=@n6, Alpha=1, ExternalForce="0,0,0") {
    angularDampingOverride=1
    angularSpringConstant=20000
    bAngularSpring=true
    bOverrideAngularDamping=true
    bUseGravityOverride=true
    gravityScale=0.35
    simulationSpace=RootRelative
    alphaBoolBlend {blendOption=Linear}
    boundBone {boneName=dyn_StringRarm_00}
    componentLinearAccScale=<1 1 1>
    componentLinearVelScale=<1 1 1>
    gravityOverride {z=-980}
    physicsBodyDefinitions [{
      boundBone {boneName=dyn_StringRarm_00}
      boxExtents=<10 30 30>
      constraintSetup {angularLimitsMax=<0 15 15>, angularLimitsMin=<0 -15 -15>, angularTarget=<1 0 0>}
      localJointOffset=<1 0 0>
     }]
   }
   n8 = Anim Dynamics - Bone: dyn_StringRarm_01(ComponentPose=@n7, Alpha=1, ExternalForce="0,0,0") {
    angularDampingOverride=1
    angularSpringConstant=20000
    bAngularSpring=true
    bOverrideAngularDamping=true
    bUseGravityOverride=true
    gravityScale=0.35
    simulationSpace=RootRelative
    alphaBoolBlend {blendOption=Linear}
    boundBone {boneName=dyn_StringRarm_01}
    componentLinearAccScale=<1 1 1>
    componentLinearVelScale=<1 1 1>
    gravityOverride {z=-980}
    physicsBodyDefinitions [{
      boundBone {boneName=dyn_StringRarm_01}
      boxExtents=<10 30 30>
      constraintSetup {angularLimitsMax=<0 15 15>, angularLimitsMin=<0 -15 -15>, angularTarget=<1 0 0>}
      localJointOffset=<1 0 0>
     }]
   }
   n9 = RigidBody!(ComponentPose=@n8, Alpha, OverrideWorldGravity:Vector="0,0,0", ExternalForce) -> (Pose) {
    bClampLinearTranslationLimitToRefPose=true
    bOverrideWorldGravity=true
    overridePhysicsAsset=&/M_MED_PoolSwirl_R_Pants_Physics
    simulationSpace=BaseBoneSpace
    alphaBoolBlend {blendOption=Linear}
    baseBoneRef {boneName=calf_r}
    componentAppliedLinearAccClamp=<3000 3000 3000>
    componentLinearAccScale=<0.32 0.32 0.32>
    componentLinearVelScale=<0.32 0.32 0.32>
    simSpaceSettings {
     maxAngularAcceleration=1000
     maxAngularVelocity=3
     maxLinearAcceleration=700
     maxLinearVelocity=500
     velocityScaleZ=0
     worldAlpha=0.2
     externalLinearDragV=<0.2 0.2>
    }
   }
   n10 = RigidBody(ComponentPose=@n9, OverrideWorldGravity="0,0,0") {
    bClampLinearTranslationLimitToRefPose=true
    bOverrideWorldGravity=true
    overridePhysicsAsset=&/M_MED_PoolSwirl_L_Pants_Physics
    simulationSpace=BaseBoneSpace
    alphaBoolBlend {blendOption=Linear}
    baseBoneRef {boneName=calf_l}
    componentAppliedLinearAccClamp=<3000 3000 3000>
    componentLinearAccScale=<0.32 0.32 0.32>
    componentLinearVelScale=<0.32 0.32 0.32>
    simSpaceSettings {
     maxAngularAcceleration=1000
     maxAngularVelocity=3
     maxLinearAcceleration=700
     maxLinearVelocity=500
     velocityScaleZ=0
     worldAlpha=0.2
     externalLinearDragV=<0.2 0.2>
    }
   }
   n11 = Get GravityOverride()
   n12 = RigidBody(ComponentPose=@n10, OverrideWorldGravity=@n11) {
    bClampLinearTranslationLimitToRefPose=true
    bOverrideWorldGravity=true
    overridePhysicsAsset=&/M_MED_PoolSwirl_Physics
    simulationSpace=BaseBoneSpace
    alphaBoolBlend {blendOption=Linear}
    baseBoneRef {boneName=pelvis}
    componentAppliedLinearAccClamp=<3000 3000 3000>
    componentLinearAccScale=<0.3 0.2 0.2>
    componentLinearVelScale=<0.3 0.2 0.2>
    simSpaceSettings {
     maxAngularAcceleration=1000
     maxAngularVelocity=3
     maxLinearAcceleration=700
     maxLinearVelocity=500
     velocityScaleZ=0
     worldAlpha=0.2
     externalLinearDragV=<0.2 0.2>
    }
   }
   n13 = RigidBody(ComponentPose=@n12, OverrideWorldGravity=@n11) {
    bClampLinearTranslationLimitToRefPose=true
    bOverrideWorldGravity=true
    overridePhysicsAsset=&/M_MED_PoolSwirl_Sleeve_Physics
    simulationSpace=BaseBoneSpace
    alphaBoolBlend {blendOption=Linear}
    baseBoneRef {boneName=upperarm_l}
    componentAppliedLinearAccClamp=<3000 3000 3000>
    componentLinearAccScale=<0.3 0.2 0.2>
    componentLinearVelScale=<0.3 0.2 0.2>
    simSpaceSettings {
     maxAngularAcceleration=1000
     maxAngularVelocity=3
     maxLinearAcceleration=700
     maxLinearVelocity=500
     velocityScaleZ=0
     worldAlpha=0.2
     externalLinearDragV=<0.2 0.2>
    }
   }
   n14 = RigidBody(ComponentPose=@n13, OverrideWorldGravity=@n11) {
    bClampLinearTranslationLimitToRefPose=true
    bOverrideWorldGravity=true
    overridePhysicsAsset=&/M_MED_PoolSwirl_Sword_Physics
    simulationSpace=BaseBoneSpace
    alphaBoolBlend {blendOption=Linear}
    baseBoneRef {boneName=pelvis}
    componentAppliedLinearAccClamp=<3000 3000 3000>
    componentLinearAccScale=<0.35 0.35 0.25>
    componentLinearVelScale=<0.35 0.35 0.25>
    simSpaceSettings {
     maxAngularAcceleration=1000
     maxAngularVelocity=3
     maxLinearAcceleration=700
     maxLinearVelocity=500
     velocityScaleZ=0
     worldAlpha=0.2
     externalLinearDragV=<0.2 0.2>
    }
   }
   Save cached pose 'Dynamics Setup'(Pose=Component To Local(ComponentPose=@n14)) {cacheName=Dynamics Setup}
   n15 = "Transform (Modify) Bone - Bone: base_sword"(
    ComponentPose=Local To Component(LocalPose=Use cached pose 'Start Pose'())
    Alpha=Get Moving Fwd or TacticalSprint() -> (Moving Fwd or TacticalSprint:double)
    Translation="0,0,0"
    Rotation="-22.002,-25.315,10.047"
    Scale="1,1,1"
   ) {
    rotationMode=BMM_Additive
    rotationSpace=BCS_BoneSpace
    alphaBoolBlend {blendOption=Linear}
    alphaScaleBiasClamp {bInterpResult=true}
    boneToModify {boneName=base_sword}
    rotation=<rot -22.002 -25.315 10.047>
   }
   n16 = "Transform (Modify) Bone - Bone: base_sword"(
    ComponentPose=@n15
    Alpha=Get MovingStanding() -> (MovingStanding:double)
    Translation="0,0,0"
    Rotation="-16.304,-18.879,5.483"
    Scale="1,1,1"
   ) {
    rotationMode=BMM_Additive
    rotationSpace=BCS_BoneSpace
    alphaBoolBlend {blendOption=Linear}
    alphaScaleBiasClamp {bInterpResult=true, interpSpeedIncreasing=2}
    boneToModify {boneName=base_sword}
    rotation=<rot -16.304 -18.879 5.483>
   }
   n17 = "Transform (Modify) Bone - Bone: base_frontPad_01"(
    ComponentPose=@n16
    Alpha=Get Crouching Idle() -> (Crouching Idle:double)
    Translation="-0.000,-2.356,0.202"
    Rotation="0.011,10.703,-0.001"
    Scale="1,1,0.95"
   ) {
    rotationMode=BMM_Additive
    rotationSpace=BCS_BoneSpace
    scaleMode=BMM_Additive
    scaleSpace=BCS_BoneSpace
    translationMode=BMM_Additive
    translationSpace=BCS_BoneSpace
    alphaBoolBlend {blendOption=Linear}
    alphaScaleBiasClamp {bInterpResult=true}
    boneToModify {boneName=base_frontPad_01}
    rotation=<rot 0.011 10.703 -0.001>
    scale {z=0.95}
    translation {y=-2.356, z=0.202}
   }
   n18 = "Transform (Modify) Bone - Bone: base_sword"(
    ComponentPose=@n17
    Alpha="To Float (Boolean)"(InBool=Get isWearingCape() -> (isWearingCape:bool))
    Translation="0,0,0"
    Rotation="20.757,0.122,-0.019"
    Scale="1,1,1"
   ) {
    rotationMode=BMM_Additive
    rotationSpace=BCS_BoneSpace
    translationMode=BMM_Additive
    translationSpace=BCS_BoneSpace
    alphaBoolBlend {blendOption=Linear}
    boneToModify {boneName=base_sword}
    rotation=<rot 20.757 0.122 -0.019>
   }
   n19 = "Transform (Modify) Bone - Bone: base_beltStrap_00"(
    ComponentPose=@n18
    Alpha=Get Crouch Idle RBAN Off() -> (Crouch Idle RBAN Off:double)
    Translation="0,0,0"
    Rotation="-19.465,-27.993,1.628"
    Scale="1,1,1"
   ) {
    rotationMode=BMM_Additive
    rotationSpace=BCS_BoneSpace
    translationMode=BMM_Additive
    translationSpace=BCS_BoneSpace
    alphaBoolBlend {blendOption=Linear}
    alphaScaleBiasClamp {bInterpResult=true}
    boneToModify {boneName=base_beltStrap_00}
    rotation=<rot -19.465 -27.993 1.628>
   }
   n20 = "Transform (Modify) Bone - Bone: base_beltStrap_00"(
    ComponentPose=@n19
    Alpha=Get Crouch RBAN Off() -> (Crouch RBAN Off:double)
    Translation="0,0,0"
    Rotation="-6.744,-17.501,-2.122"
    Scale="1,1,1"
   ) {
    rotationMode=BMM_Additive
    rotationSpace=BCS_BoneSpace
    translationMode=BMM_Additive
    translationSpace=BCS_BoneSpace
    alphaBoolBlend {blendOption=Linear}
    alphaScaleBiasClamp {bInterpResult=true}
    boneToModify {boneName=base_beltStrap_00}
    rotation=<rot -6.744 -17.501 -2.122>
   }
   Save cached pose 'Pre Dynamics'(Pose=Component To Local(ComponentPose=@n20)) {cacheName=Pre Dynamics}
   n21 = "Transform (Modify) Bone - Bone: base_backPad_01"(
    ComponentPose=Local To Component(LocalPose=Use cached pose 'Dynamics Setup'())
    Alpha=1
    Translation="0,0,0"
    Rotation="0,-15,0"
    Scale="1,1,1"
   ) {
    rotationMode=BMM_Additive
    rotationSpace=BCS_BoneSpace
    translationMode=BMM_Additive
    translationSpace=BCS_BoneSpace
    alphaBoolBlend {blendOption=Linear}
    alphaScaleBiasClamp {bInterpResult=true}
    boneToModify {boneName=base_backPad_01}
    rotation {yaw=-15}
   }
   n22 = "Transform (Modify) Bone - Bone: base_backPad_01"(ComponentPose=@n21, Alpha=Get Crouching Idle(), Translation="0,0,0", Rotation="0,-5,0", Scale="1,1,1") {
    rotationMode=BMM_Additive
    rotationSpace=BCS_BoneSpace
    translationMode=BMM_Additive
    translationSpace=BCS_BoneSpace
    alphaBoolBlend {blendOption=Linear}
    alphaScaleBiasClamp {bInterpResult=true}
    boneToModify {boneName=base_backPad_01}
    rotation {yaw=-5}
   }
   n23 = "Transform (Modify) Bone - Bone: shoulderPad_ExtraPivot_l"(ComponentPose=@n22, Alpha=1, Translation="2,1,-1", Rotation="0,0,20", Scale="1,1,1") {
    rotationMode=BMM_Additive
    rotationSpace=BCS_BoneSpace
    translationMode=BMM_Additive
    translationSpace=BCS_BoneSpace
    alphaBoolBlend {blendOption=Linear}
    alphaScaleBiasClamp {bInterpResult=true}
    boneToModify {boneName=shoulderPad_ExtraPivot_l}
    rotation {roll=20}
    translation=<2 1 -1>
   }
   n24 = "Transform (Modify) Bone - Bone: shoulderPad_ExtraPivot_l"(ComponentPose=@n23, Alpha=1, Translation="0,0,1", Rotation="-4,0,10", Scale="1,1,1") {
    rotationMode=BMM_Additive
    rotationSpace=BCS_BoneSpace
    translationMode=BMM_Additive
    translationSpace=BCS_BoneSpace
    alphaBoolBlend {blendOption=Linear}
    alphaScaleBiasClamp {bInterpResult=true}
    boneToModify {boneName=shoulderPad_ExtraPivot_l}
    rotation {pitch=-4, roll=10}
    translation {z=1}
   }
   n25 = "Transform (Modify) Bone - Bone: shoulderPad_ExtraPivot_r"(ComponentPose=@n24, Alpha=1, Translation="0,0,0", Rotation="0,0,0", Scale="1,1,1") {
    rotationMode=BMM_Additive
    rotationSpace=BCS_BoneSpace
    translationMode=BMM_Additive
    translationSpace=BCS_BoneSpace
    alphaBoolBlend {blendOption=Linear}
    alphaScaleBiasClamp {bInterpResult=true}
    boneToModify {boneName=shoulderPad_ExtraPivot_r}
   }
   n26 = "Transform (Modify) Bone - Bone: shoulderPad_ExtraPivot_r"(ComponentPose=@n25, Alpha=1, Translation="0,0,3", Rotation="-7,0,0", Scale="1,1,1") {
    rotationMode=BMM_Additive
    rotationSpace=BCS_BoneSpace
    translationMode=BMM_Additive
    translationSpace=BCS_BoneSpace
    alphaBoolBlend {blendOption=Linear}
    alphaScaleBiasClamp {bInterpResult=true}
    boneToModify {boneName=shoulderPad_ExtraPivot_r}
    rotation {pitch=-7}
    translation {z=3}
   }
   n27 = "Transform (Modify) Bone - Bone: base_sword"(ComponentPose=@n26, Alpha=1, Translation="4.162,0.000,3.162", Rotation="-61.540,-0.000,0.000", Scale="1,1,1") {
    rotationMode=BMM_Additive
    rotationSpace=BCS_BoneSpace
    translationMode=BMM_Additive
    translationSpace=BCS_BoneSpace
    alphaBoolBlend {blendOption=Linear}
    alphaScaleBiasClamp {bInterpResult=true}
    boneToModify {boneName=base_sword}
    rotation=<rot -61.54 0 0>
    translation {x=4.162, z=3.162}
   }
   n28 = "Transform (Modify) Bone - Bone: base_sword"(ComponentPose=@n27, Alpha=1, Translation="5.971,0.000,4.698", Rotation="-55.679,-0.000,0.000", Scale="1,1,1") {
    rotationMode=BMM_Additive
    rotationSpace=BCS_BoneSpace
    translationMode=BMM_Additive
    translationSpace=BCS_BoneSpace
    alphaBoolBlend {blendOption=Linear}
    alphaScaleBiasClamp {bInterpResult=true}
    boneToModify {boneName=base_sword}
    rotation=<rot -55.679 0 0>
    translation {x=5.971, z=4.698}
   }
   Save cached pose 'Post Correctives'(Pose=Component To Local(ComponentPose=@n28)) {cacheName=Post Correctives}
  ]}
 EventGraph {nodes [
   Event BlueprintUpdateAnimation()
   > n1 = Sequence()
   > n2 = Gravity Override State Function DEPRECATED(CharacterPartAnimInstance=Self-Reference(), GameTimeInSeconds=GetGameTimeInSeconds(), GravityOverrideParams=Get Gravity Override Params())
   > Set GravityOverride(GravityOverride=@n2)
   n3 = PlayerMovement(execute=@n1.then_1, AnimInstance=Self-Reference(), XDotProductCoefficient=1, YDotProductCoefficient=1, ZDotProductCoefficient=1)
   n4 = BreakVector(InVec=MakeVector(X=@n3, Y=@n3.YMovement))
   > Set Left(Left:double="To Float (Boolean)"(InBool="float < float"(A=@n4, B=-50))) -> (Output_Get:double)
   > Set Right(Right:double="To Float (Boolean)"(InBool="float > float"(A=@n4, B=50))) -> (Output_Get:double)
   > Set Backward(Backward:double="To Float (Boolean)"(InBool="float < float"(A=@n4.Y, B=-40))) -> (Output_Get:double)
   > Set Forward(Forward:double="To Float (Boolean)"(InBool="float > float"(A=@n4.Y, B=40))) -> (Output_Get:double)
   > Set is Crouching CR(is Crouching CR:double="To Float (Boolean)"(InBool=Get bIsCrouching())) -> (Output_Get:double)
   n5 = Get bIsMoving2D()
   n6 = NOT Boolean(A=@n5)
   n7 = Get bIsCrouching()
   > Set Crouching Idle(Crouching Idle:double="To Float (Boolean)"(InBool=AND Boolean(A=@n7, B=@n6))) -> (Output_Get:double)
   n8 = NOT Boolean(A=Get bRigidBodyEnabled())
   > Set Crouch Idle RBAN Off(Crouch Idle RBAN Off:double="To Float (Boolean)"(InBool=AND Boolean!(A=@n6, B=@n8, C:bool=@n7) -> (ReturnValue))) -> (Output_Get:double)
   > Set Crouch RBAN Off(Crouch RBAN Off:double="To Float (Boolean)"(InBool=AND Boolean(A=@n8, B=@n7, C=@n5))) -> (Output_Get:double)
   > Set Moving Fwd or TacticalSprint(Moving Fwd or TacticalSprint:double="To Float (Boolean)"(InBool=Get bIsTacticalSprinting())) -> (Output_Get:double)
   n9 = NOT Boolean(A=@n7)
   n10 = Get bIsMoving2D()
   > Set MovingStanding(MovingStanding:double="To Float (Boolean)"(InBool=AND Boolean!(A=@n10, B=@n9) -> (ReturnValue))) -> (Output_Get:double)
   n11 = NOT Boolean(A=Get bIsPlayingEmote())
   n12 = Get bIsInFrontEnd()
   > Set FrontendIdle(FrontendIdle:double="To Float (Boolean)"(InBool=AND Boolean(A=@n12, B=@n11))) -> (Output_Get:double)
   Event OnCharacterCustomizationLoaded()
   > n13 = Sequence()
   Set isWearingCape(execute=@n13.then_1, isWearingCape=CheckForBackbling().bIsBackblingACape)
   n14 = Branch(
    execute=[@n16.then, @n13.then_0]
    Condition=PickaxeItemDefinitionHasAnyMetaTag(Tags=Make Array("[0]"=(TagName="Cosmetics.MetaTags.PoolSwirl")), bRequiresEquipped=true)
   )
   > n15 = Get Dynamic Material Instance(SkeletalMeshComponent=GetPartSkeletalMeshComponent())
   > SetScalarParameterValue(self=@n15, ParameterName=Hide Element 02, Value=1)
   SetScalarParameterValue(execute=@n14.else, self=@n15, ParameterName=Hide Element 02, Value=0)
   n16 = Event OnWeaponChanged()
  ]}
]
variables [
 Backward {type=number}
 Crouch Idle RBAN Off {type=number}
 Crouch RBAN Off {type=number}
 Crouching Idle {type=number}
 Forward {type=number}
 FrontendIdle {type=number}
 Gravity Override Params {type=GravityOverrideParamsStruct}
 GravityOverride {type=Vector}
 is Crouching CR {type=number}
 is Jump {type=boolean}
 isWearingBackbling {type=boolean}
 isWearingCape {type=boolean}
 Left {type=number}
 Moving Fwd or TacticalSprint {type=number}
 MovingStanding {type=number}
 Right {type=number}
]
