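# Physics asset for the face accessory part of the male medium PoolSwirl player cosmetic, covering its front hair.
# Capsule bodies on head, base_hairsFront and dyn_hairsFront form a short chain from the head to the dynamic hair bone.
# The base-to-head joint leaves swing and twist free, while the dynamic-to-base joint has narrow, uneven swing limits and almost no twist.
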PhysicsAsset /BRCosmetics/Characters/Player/Male/Medium/Bodies/M_MED_PoolSwirl/Meshes/Parts/M_MED_PoolSwirl_FaceAcc_Physics
bodies {
 base_hairsFront [{shapeType=Capsule, length=0.505, radius=0.505, shapeName=base_hairsFront_capsule, rotation=<rot 0 -90 0>}]
 dyn_hairsFront [{shapeType=Capsule, length=2, radius=1, shapeName=dyn_hairsFront_capsule, center=<2 0 0>, rotation=<rot 90 0 0>}]
 head [{shapeType=Capsule, length=2, radius=2, shapeName=head_capsule}]
}
constraints [
 {bone1Name=dyn_hairsFront, bone2Name=base_hairsFront, swing1LimitDegrees=10, swing2LimitDegrees=30, twistLimitDegrees=1}
 {
  bone1Name=base_hairsFront
  bone2Name=head
  swing1LimitDegrees=45
  swing1Motion=Free
  swing2LimitDegrees=45
  swing2Motion=Free
  twistLimitDegrees=45
  twistMotion=Free
 }
]
